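# Niagara swing trail for the BakerStep III drum pickaxe, drawing short-lived ribbons from sockets on the weapon's skeletal mesh.
# Three active emitters (HardEdge, SoftA, SoftB) share FN anim-trail modules and layer different widths, with SoftA adding a socket-rotated offset and sprite renderer.
# Alpha fades over particle life, with time-of-day scaling of opacity and tint.
# A disabled Chains emitter retains an eroding ribbon variant driven through dynamic material parameters.
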NiagaraSystem /SparksCosmetics/Instrument/Drum/BakerStep/Effects/NS_Pickaxe_Drum_BakerStep_III_Trail
dependencies {
 usedDataInterfaces [&NiagaraDataInterfaceCurve, &NiagaraDataInterfaceSocketReader]
 usedDynamicInputs []
 usedModules [
  &SystemState
  &self:SystemSpawnScript.SystemScriptSource.SystemScriptGraph.NiagaraNodeAssignment_0.SetVariables_#1
 ]
}
emitters {
 Chains {
  bEnabled=false
  emitterSpawnScript [/SetVariables(Emitter.CustomSeed:NiagaraInt32=&RandomRangeInt(Minimum=0, Maximum=2000)) {isSetParameters=true}]
  emitterUpdateScript [
   &EmitterState(Life Cycle Mode=::NewEnumerator1, Loop Behavior=::NewEnumerator1, Loop Duration=0.27, UseLoopDelay=false)
   &FN_SpawnAnimTrail()
  ]
  particleSpawnScript [
   &FN_InitializeAnimTrailBoneIndecies(First Bone or Socket Index=1)
   &FN_InitializeAnimTrail(
    AnimTrailWidth=8
    Color=&MultiplyLinearColor(
     Linear Color=<rgba 0.7656 0.7656 0.7656 0.5>
     Scale Factor=&MakeVector4_FromLinearColor(LinearColor=$System.TOD)
    )
    Lifetime=0.2
    Position Offset=&RotateVectorByQuaternion(
     Vector To Rotate=<0 0 10>
     Quaternion=&self:GetSocketRotation(Socket Reader:NiagaraDataInterfaceSocketReader=$System.Socket Reader, SocketIndex:NiagaraInt32=1)
    )
   )
   &DynamicMaterialParameters(
    Index 0 Param 1=1
    Index 0 Param 2=5
    Index 0 Param 3=3
    Index 0 Param 4=1
    Index 1 Param 1=3
    Index 1 Param 3=&FloatFromCurve(FloatCurve=DataInterface() {propertyValues {Curve {keys [<at 0 1>,<at 0.02 1>]}}}, CurveIndex=$Engine.System.Age)
    Module.Index0Float0Write=true
    Module.Index0Float1Write=true
    Module.Index0Float2Write=true
    Module.Index0Float3Write=true
    Module.Index1Float0Write=true
    Module.Index1Float1Write=false
    Module.Index1Float2Write=false
    Module.Index1Float3Write=false
    Write Parameter Index 1=::NewEnumerator1
   )
   &AddVelocity(
    Velocity=&RandomRangeVector(Minimum=<-1 -1 -1>, Maximum=<1 1 1>)
    Velocity Speed Scale=&FloatFromCurve(
     FloatCurve=DataInterface() {propertyValues {Curve {keys [{arriveTangent=-1.7333, interpMode=RCIM_Cubic, leaveTangent=-1.7333, tangentMode=RCTM_User, time=0, value=1}, <at 0.27 0>]}}}
     CurveIndex=$System.Age
     Scale Curve=100
    )
   )
   /SetVariables(Particles.InitialErosion:NiagaraFloat=&FloatFromCurve(FloatCurve=DataInterface() {propertyValues {Curve {keys [<at 0 1>,<at 0.02 0>]}}}, CurveIndex=$Engine.System.Age)) {isSetParameters=true}
   &InitialMeshOrientation(
    Mesh Orientation Mode=::NewEnumerator6
    Orientation Quaternion:Quat4f=&self:GetSocketRotation(Socket Reader:NiagaraDataInterfaceSocketReader=$System.Socket Reader, SocketIndex:NiagaraInt32=0)
    Use Rotation Vector=false
   ) {enabled=false}
  ]
  particleUpdateScript [
   &SolveForcesAndVelocity(Clamp Velocity=false, Limit Acceleration=false)
   &FN_UpdateAnimTrail()
   &ScaleRibbonWidth(Ribbon Width Scale=&FloatFromCurve(
     FloatCurve=DataInterface() {propertyValues {Curve {keys [
         {arriveTangent=-4.3394, interpMode=RCIM_Cubic, leaveTangent=-4.3394, tangentMode=RCTM_User, time=0.7, value=1}
         {interpMode=RCIM_Cubic, tangentMode=RCTM_User, time=1, value=0}
        ]}}}
     Scale Curve=&OneMinusFloat(Float=$Particles.InitialErosion)
    ))
   &ScaleColor(
    Scale Alpha=&Multiply_Float(
     A=&FloatFromCurve(FloatCurve=DataInterface() {propertyValues {Curve {keys [<at 0.3 1 RCIM_Cubic>,<at 1 0 RCIM_Cubic>]}}}, Scale Curve=1.5)
     B=&FNTimeOfDayScalar(Night Scale=1, Morning Scale=0.5, Evening Scale=0.5)
    )
    ScaleA=true
    ScaleRGB=true
   ) {module=ScaleColor002}
   &DynamicMaterialParameters(
    Index 1 Param 3=&Add_Float(
     A=$Particles.InitialErosion
     B=&FloatFromCurve(FloatCurve=DataInterface() {propertyValues {Curve {keys [<at 0.5 0>,<at 1 1>]}}})
    )
    Module.Index0Float0Write=false
    Module.Index0Float1Write=false
    Module.Index0Float2Write=false
    Module.Index0Float3Write=false
    Module.Index1Float0Write=false
    Module.Index1Float1Write=false
    Module.Index1Float2Write=true
    Module.Index1Float3Write=false
    Write Parameter Index 1=::NewEnumerator1
   ) {module=DynamicMaterialParameters001}
  ]
  rendererClasses [&NiagaraRibbonRendererProperties]
  renderers [{rendererClass=&NiagaraRibbonRendererProperties}]
 }
 FN_AnimTrail_HardEdge {
  emitterSpawnScript [&FN_RibbonIDOffsetUtility()]
  emitterUpdateScript [
   &EmitterState()
   &FN_SpawnAnimTrail()
  ]
  particleSpawnScript [
   &FN_InitializeAnimTrailBoneIndecies(First Bone or Socket Index=1, Second Bone or Socket Index=0)
   &FN_InitializeAnimTrail(
    AnimTrailWidth=1.2
    Color=&FNTimeOfDayColor(
     Night Color=&MultiplyLinearColor(Linear Color=<rgba 0.1683 0.1714 0.159 1>, Scale Factor=<0.65 0.65 1 1>)
     Morning Color=<rgba 0.1683 0.1714 0.159 1>
     Day Color=<rgba 0.1683 0.1714 0.159 1>
     Evening Color=<rgba 0.1683 0.1714 0.159 1>
    )
    Lifetime=0.2
   )
   &AddVelocity(Velocity=<0 400 0>) {enabled=false}
  ]
  particleUpdateScript [
   &SolveForcesAndVelocity(Clamp Velocity=false, Limit Acceleration=false)
   &FN_UpdateAnimTrail() {module=FN_UpdateAnimTrail001}
   &ScaleColor(
    Scale Alpha=&Multiply_Float(
     A=&FloatFromCurve(
      FloatCurve=DataInterface() {propertyValues {Curve {keys [
          {interpMode=RCIM_Cubic, leaveTangent=-2.3646, tangentMode=RCTM_Break, time=0, value=1}
          <at 0.6689 0.0233>
          {arriveTangent=-0.2139, interpMode=RCIM_Cubic, tangentMode=RCTM_Break, time=1, value=0}
         ]}}}
      Scale Curve=0.5
     )
     B=&FNTimeOfDayScalar(Night Scale=0.75, Morning Scale=0.25, Evening Scale=0.4)
    )
    Scale RGB=&FN_Select_Vector_from_Time_Of_Day(Morning=<1 1 1>, Day=<1 1 1>, Evening=<1 1 1>, Night=<0.8 0.8 1>)
    ScaleA=true
    ScaleRGB=true
   ) {module=ScaleColor002}
  ]
  rendererClasses [&NiagaraRibbonRendererProperties]
  renderers [{rendererClass=&NiagaraRibbonRendererProperties}]
 }
 FN_AnimTrail_SoftA {
  emitterSpawnScript [&FN_RibbonIDOffsetUtility()]
  emitterUpdateScript [
   &EmitterState()
   &FN_SpawnAnimTrail()
  ]
  particleSpawnScript [
   &FN_InitializeAnimTrailBoneIndecies(First Bone or Socket Index=1, Second Bone or Socket Index=0)
   &FN_InitializeAnimTrail(
    AnimTrailWidth=2
    Lifetime=0.2
    Position Offset=&RotateVectorByQuaternion(
     Vector To Rotate=<0 0 5>
     Quaternion=&self:GetSocketRotation(Socket Reader:NiagaraDataInterfaceSocketReader=$System.Socket Reader, SocketIndex:NiagaraInt32=1)
    )
   )
   &AddVelocity(Velocity=<0 400 0>) {enabled=false}
  ]
  particleUpdateScript [
   &SolveForcesAndVelocity(Clamp Velocity=false, Limit Acceleration=false)
   &FN_UpdateAnimTrail() {module=FN_UpdateAnimTrail001}
   &ScaleColor(
    Scale Alpha=&Multiply_Float(
     A=&FloatFromCurve(FloatCurve=DataInterface() {propertyValues {Curve {keys [
          {interpMode=RCIM_Cubic, leaveTangent=-3.5578, tangentMode=RCTM_Break, time=0, value=1}
          <at 0.6689 0.0233>
          {arriveTangent=-0.2139, interpMode=RCIM_Cubic, tangentMode=RCTM_Break, time=1, value=0}
         ]}}})
     B=&FNTimeOfDayScalar(Night Scale=0.75, Morning Scale=0.25, Evening Scale=0.4)
    )
    Scale RGB=&FN_Select_Vector_from_Time_Of_Day(Morning=<1.15 1.15 0.75>, Day=<1 1 1>, Evening=<1 1 1>, Night=<0.8 0.8 1.35>)
    ScaleA=true
    ScaleRGB=true
   ) {module=ScaleColor002}
  ]
  rendererClasses [&NiagaraRibbonRendererProperties, &NiagaraSpriteRendererProperties]
  renderers [
   {rendererClass=&NiagaraRibbonRendererProperties}
   {rendererClass=&NiagaraSpriteRendererProperties, rendererIndex=1}
  ]
 }
 FN_AnimTrail_SoftB {
  emitterSpawnScript [
   &FN_RibbonIDOffsetUtility()
   /SetVariables() {isSetParameters=true}
  ]
  emitterUpdateScript [
   &EmitterState()
   &FN_SpawnAnimTrail()
  ]
  particleSpawnScript [
   &FN_InitializeAnimTrailBoneIndecies(First Bone or Socket Index=1, Second Bone or Socket Index=0)
   &FN_InitializeAnimTrail(AnimTrailWidth=0.25, Color=<rgba 0.3573 0.3573 0.3573 1>, Lifetime=0.2)
   &AddVelocity(Velocity=<0 400 0>) {enabled=false}
  ]
  particleUpdateScript [
   &SolveForcesAndVelocity(Clamp Velocity=false, Limit Acceleration=false)
   &FN_UpdateAnimTrail() {module=FN_UpdateAnimTrail001}
   &ScaleRibbonWidth(Ribbon Width Scale=&FloatFromCurve(FloatCurve=DataInterface() {propertyValues {Curve {keys [
         {arriveTangent=-1.2579, interpMode=RCIM_Cubic, time=0, value=0}
         <at 0.2 1 RCIM_Cubic>
         {arriveTangent=-0.9664, interpMode=RCIM_Cubic, leaveTangent=-0.9664, tangentMode=RCTM_User, time=1, value=0}
        ]}}}))
   &ScaleColor(
    Scale Alpha=&Multiply_Float(
     A=&FloatFromCurve(FloatCurve=DataInterface() {propertyValues {Curve {keys [
          {interpMode=RCIM_Cubic, leaveTangent=-3.5578, tangentMode=RCTM_Break, time=0, value=1}
          <at 0.6689 0.0233>
          {arriveTangent=-0.2139, interpMode=RCIM_Cubic, tangentMode=RCTM_Break, time=1, value=0}
         ]}}})
     B=&FNTimeOfDayScalar(Night Scale=0.75, Morning Scale=0.25, Evening Scale=0.4)
    )
    Scale RGB=&FN_Select_Vector_from_Time_Of_Day(Morning=<1 1 1>, Day=<1 1 1>, Evening=<1 1 1>, Night=<0.8 0.8 1>)
    ScaleA=true
    ScaleRGB=true
   ) {module=ScaleColor002}
  ]
  rendererClasses [&NiagaraRibbonRendererProperties]
  renderers [{rendererClass=&NiagaraRibbonRendererProperties}]
 }
}
summary {
 systemName=NS_Pickaxe_Drum_BakerStep_III_Trail
 emitters [
  {
   emitterName=FN_AnimTrail_SoftA
   rendererClasses [&NiagaraRibbonRendererProperties, &NiagaraSpriteRendererProperties]
  }
  {emitterName=FN_AnimTrail_HardEdge, rendererClasses [&NiagaraRibbonRendererProperties]}
  {emitterName=FN_AnimTrail_SoftB, rendererClasses [&NiagaraRibbonRendererProperties]}
  {bEnabled=false, emitterName=Chains, rendererClasses [&NiagaraRibbonRendererProperties]}
 ]
 userVariables [
  User.RibbonIDOffset {type=NiagaraInt32}
  User.SkeletalMesh {
   type=NiagaraDataInterfaceSkeletalMesh
   defaultValue {
    struct=NiagaraExt_VariableValue_DataInterface
    value {dataInterface=&self:NiagaraDataInterfaceSkeletalMesh_0}
   }
  }
 ]
}
